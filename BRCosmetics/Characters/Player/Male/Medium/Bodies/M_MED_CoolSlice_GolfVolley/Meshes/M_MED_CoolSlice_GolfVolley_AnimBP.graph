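# Animation Blueprint for the CoolSlice GolfVolley male medium body cosmetic, copying the attached parent mesh's pose through the body's Control Rig.
# Bone modifiers then additively offset the right calf twist and replace neck and head transforms in parent-bone space.
# On customization load it stores the Head part; each update detects attachment via C_brow_mid and caches that mesh's neck and head transforms in variables, with no visible wiring to the modifiers.
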AnimBlueprint /BRCosmetics/Characters/Player/Male/Medium/Bodies/M_MED_CoolSlice_GolfVolley/Meshes/M_MED_CoolSlice_GolfVolley_AnimBP
target_skeleton=&/M_MED_CoolSlice_GolfVolley_Skeleton
graphs [
 AnimGraph {nodes [
   n1 = Control Rig(Source=Copy Pose From Mesh() {bUseAttachedParent=true}) {controlRigAssetReference {blueprintRigClass=&/M_MED_CoolSlice_GolfVolley_CtrlRig.M_MED_CoolSlice_GolfVolley_CtrlRig_C}}
   n2 = "Transform (Modify) Bone - Bone: calf_twist_02_r"(ComponentPose=Local To Component(LocalPose=@n1), Alpha=1, Translation="8,3,1", Rotation="0,0,0", Scale="1,1,1") {
    rotationMode=BMM_Additive
    rotationSpace=BCS_BoneSpace
    translationMode=BMM_Additive
    translationSpace=BCS_BoneSpace
    boneToModify {boneName=calf_twist_02_r}
    translation=<8 3 1>
   }
   n3 = "Transform (Modify) Bone - Bone: neck_01"!(ComponentPose=@n2, bAlphaBoolEnabled:bool=True, Translation="0,0,0", Rotation="0,0,0") -> (Pose) {
    alphaInputType=Bool
    rotationMode=BMM_Replace
    rotationSpace=BCS_ParentBoneSpace
    translationMode=BMM_Replace
    translationSpace=BCS_ParentBoneSpace
    alphaBoolBlend {blendOption=Linear}
    boneToModify {boneName=neck_01}
   }
   n4 = "Transform (Modify) Bone - Bone: neck_02"!(ComponentPose=@n3, bAlphaBoolEnabled:bool=True, Translation="0,0,0", Rotation="0,0,0") -> (Pose) {
    alphaInputType=Bool
    rotationMode=BMM_Replace
    rotationSpace=BCS_ParentBoneSpace
    translationMode=BMM_Replace
    translationSpace=BCS_ParentBoneSpace
    alphaBoolBlend {blendOption=Linear}
    boneToModify {boneName=neck_02}
   }
   n5 = "Transform (Modify) Bone - Bone: head"!(ComponentPose=@n4, bAlphaBoolEnabled:bool=True, Translation="0,0,0", Rotation="0,0,0") -> (Pose) {
    alphaInputType=Bool
    rotationMode=BMM_Replace
    rotationSpace=BCS_ParentBoneSpace
    translationMode=BMM_Replace
    translationSpace=BCS_ParentBoneSpace
    alphaBoolBlend {blendOption=Linear}
    boneToModify {boneName=head}
   }
   Output Pose(Result=Component To Local(ComponentPose=@n5)) {layerGroup=None}
  ]}
 EventGraph {nodes [
   Event BlueprintUpdateAnimation()
   n1 = Get HeadMesh() -> (HeadMesh:Skeletal Mesh Component Budgeted Object Reference)
   > Is Valid(InputObject=@n1)
   n2 = GetSocketBoneName!(self:Skinned Mesh Component Object Reference=@n1, InSocketName:Name=C_brow_mid) -> (ReturnValue:Name=None)
   > n3 = Set isHeadAttached(isHeadAttached:bool="Not Equal Exactly (String)"(A="To String (Name)"(InName=@n2), B=None)) -> (Output_Get:bool)
   > Branch(Condition=@n3)
   n4 = Get HeadMesh()
   n5 = BreakTransform(InTransform=GetSocketTransform(self=@n4, InSocketName=neck_01, TransformSpace=RTS_ParentBoneSpace))
   > Set neck_01_location(neck_01_location:Vector=@n5) -> (Output_Get:Vector)
   > Set neck_01_rotation(neck_01_rotation:Rotator=@n5.Rotation) -> (Output_Get:Rotator)
   n6 = BreakTransform(InTransform=GetSocketTransform(self=@n4, InSocketName=neck_02, TransformSpace=RTS_ParentBoneSpace))
   > Set neck_02_location(neck_02_location:Vector=@n6) -> (Output_Get:Vector)
   > Set neck_02_rotation(neck_02_rotation:Rotator=@n6.Rotation) -> (Output_Get:Rotator)
   n7 = BreakTransform(InTransform=GetSocketTransform(self=@n4, InSocketName=head, TransformSpace=RTS_ParentBoneSpace))
   > Set head_location(head_location:Vector=@n7) -> (Output_Get:Vector)
   > Set head_rotation(head_rotation:Rotator=@n7.Rotation) -> (Output_Get:Rotator)
   n8 = Event OnCharacterCustomizationLoaded()
   > Set HeadMesh(HeadMesh:Skeletal Mesh Component Budgeted Object Reference=Get Array Item with EFort Custom Part Type(Array=@n8.CustomMeshParts, PartType=Head)) -> (Output_Get:Skeletal Mesh Component Budgeted Object Reference)
  ]}
]
variables [
 head_location {type=Vector}
 head_rotation {type=Rotator}
 HeadMesh {type=SkeletalMeshComponentBudgeted}
 isHeadAttached {type=boolean}
 neck_01_location {type=Vector}
 neck_01_rotation {type=Rotator}
 neck_02_location {type=Vector}
 neck_02_rotation {type=Rotator}
]
